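; Burst gameplay cue for a vehicle passing through a flaming hoop, tailoring the launch audio to the local viewer.
; It casts the target to a Fort vehicle, gets the driver's team affiliation to the local player, and loops over passengers to check for the locally viewed pawn.
; Those results set the IsEnemy and IsLocal parameters on the first spawned audio component, with only hostile drivers counting as enemy.
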(fn OnBurst(EitherVFX-preferthisversion)
  (params (MyTarget :type Actor%20Object%20Reference) (Parameters :type Gameplay%20Cue%20Parameters%20Structure%20%28by%20ref%29) (ParticleComponents :type Array%20of%20FXSystem%20Component%20Object%20References) (AudioComponents :type Array%20of%20Audio%20Component%20Object%20References) (BurstCameraShakeInstance :type Camera%20Shake%20Base%20Object%20Reference) (BurstDecalInstance :type Decal%20Actor%20Object%20Reference))
  (|CallParentFunction MyTarget Parameters ParticleComponents AudioComponents BurstCameraShakeInstance BurstDecalInstance)
  (Utilities|FlowControl|Sequence
    (:then_0
      ;[0 0.1483 0.349] Get the affiliation between the vehicle's driver and locally viewed player. Get if the locally viewed player is a passenger in the launched vehicle.
      (bind _asfort_athena_vehicle (Utilities|Casting|CastToFortAthenaVehicle (Variables|GetMyTarget)))
      (bind _output_get (LocalVariables|SetFort%20Vehicle _asfort_athena_vehicle))
      (LocalVariables|SetDriver%20Affiliation (Team|GetTeamAffiliationforActortoLocalPlayer (Vehicle|GetDriver _output_get) true true))
      (bind _array_element (Utilities|Array|ForEachLoopwithBreak (Vehicle|GetAllPassengers (LocalVariables|GetFort%20Vehicle)))
        (:LoopBody
          (if (Pawn|IsLocallyViewed _array_element)
            (LocalVariables|SetIs%20Passenger true))
          (LocalVariables|SetIs%20Passenger false))))
    (:then_1
      (Utilities|FlowControl|Sequence
        (:then_0
          ;[0.1615 0.1212 0.08914] Pass values into audio component
          (bind _output_get_2 (LocalVariables|SetLaunch%20Audio (Utilities|Array|Get(acopy) (Variables|GetAudioComponents))))
          (Utilities|IsValid _output_get_2
            (:Is%20Valid
              (Audio|Parameter|SetBooleanParameter (LocalVariables|GetLaunch%20Audio) "IsEnemy" (Utilities|Select :Index (LocalVariables|GetDriver%20Affiliation) :Friendly false :Neutral false :Hostile true) :!class /Script/Engine.AudioComponent)
              (Audio|Parameter|SetBooleanParameter (LocalVariables|GetLaunch%20Audio) "IsLocal" (LocalVariables|GetIs%20Passenger) :!class /Script/Engine.AudioComponent))))
        (:then_1
          (Utilities|FlowControl|Sequence
            (:then_0
              (return))))))))
